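# Donut crater decal material that draws a noise-broken radial frosting shape on surfaces, with inner and outer distances that can shrink over time.
# The shape mask feeds opacity, banding and frosting normals, which blend with scene normals before MF_ColorBanding stylizes the color.
# Opacity fades by projection angle, using world-position derivatives, and by surface slope on higher-end platforms.
# Time-of-day colors and sky light direction build unlit emission, and a material proxy supplies a simpler version.
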Material /SonicSpearGameplay/Assets/Materials/DebrisMaterials/M_DonutCrater_Decal
graph {
 expressions {
  MaterialExpressionAbs_0 {class=MaterialExpressionAbs, inputs [None {from=MaterialExpressionDotProduct_2, output=""}], properties {}}
  MaterialExpressionAbs_2 {class=MaterialExpressionAbs, inputs [None {from=MaterialExpressionComponentMask_7, output=""}], properties {}}
  MaterialExpressionActorPositionWS_0 {class=MaterialExpressionActorPositionWS, inputs [], properties {}}
  MaterialExpressionAdd_0 {
   class=MaterialExpressionAdd
   inputs [
    A {from=MaterialExpressionMultiply_11}
    B {from=MaterialExpressionDivide_0}
   ]
  }
  MaterialExpressionAdd_10 {
   class=MaterialExpressionAdd
   inputs [
    A {from=MaterialExpressionDotProduct_4}
    B {from=""}
   ]
  }
  MaterialExpressionAdd_11 {
   class=MaterialExpressionAdd
   inputs [
    A {from=MaterialExpressionMultiply_39}
    B {from=MaterialExpressionMultiply_40}
   ]
   properties {desc=B}
  }
  MaterialExpressionAdd_12 {
   class=MaterialExpressionAdd
   inputs [
    A {from=MaterialExpressionAbs_2}
    B {from=""}
   ]
   properties {constB=-0.5}
  }
  MaterialExpressionAdd_13 {
   class=MaterialExpressionAdd
   inputs [
    A {from=MaterialExpressionNamedRerouteUsage_13}
    B {from=MaterialExpressionScalarParameter_25}
   ]
  }
  MaterialExpressionAdd_5 {
   class=MaterialExpressionAdd
   inputs [
    A {from=MaterialExpressionDotProduct_1}
    B {from=MaterialExpressionScalarParameter_21}
   ]
   properties {constB=0}
  }
  MaterialExpressionAdd_6 {
   class=MaterialExpressionAdd
   inputs [
    A {from=MaterialExpressionMultiply_37}
    B {from=MaterialExpressionMultiply_38}
   ]
   properties {desc=T}
  }
  MaterialExpressionAdd_7 {
   class=MaterialExpressionAdd
   inputs [
    A {from=MaterialExpressionDivide_1}
    B {from=""}
   ]
  }
  MaterialExpressionAdd_8 {
   class=MaterialExpressionAdd
   inputs [
    A {from=MaterialExpressionMultiply_26}
    B {from=MaterialExpressionNamedRerouteUsage_9}
   ]
  }
  MaterialExpressionAdd_9 {
   class=MaterialExpressionAdd
   inputs [
    A {from=MaterialExpressionMultiply_27}
    B {from=""}
   ]
  }
  MaterialExpressionBlendMaterialAttributes_0 {
   class=MaterialExpressionBlendMaterialAttributes
   inputs [
    A {from=MaterialExpressionSetMaterialAttributes_2, output=""}
    B {from=MaterialExpressionSetMaterialAttributes_3, output=""}
    Alpha {from="", output=""}
   ]
   properties {}
  }
  MaterialExpressionBreakMaterialAttributes_0 {class=MaterialExpressionBreakMaterialAttributes, inputs [Attr {from=MaterialExpressionMaterialFunctionCall_6, output=Result}], properties {}}
  MaterialExpressionCeil_0 {class=MaterialExpressionCeil, inputs [None {from=MaterialExpressionSaturate_9, output=""}], properties {}}
  MaterialExpressionCeil_1 {class=MaterialExpressionCeil, inputs [None {from=MaterialExpressionMax_3, output=""}], properties {}}
  MaterialExpressionClamp_0 {
   class=MaterialExpressionClamp
   inputs [
    None {from=MaterialExpressionLinearInterpolate_8, output=""}
    Min {from=MaterialExpressionScalarParameter_28, output=""}
    Max {from="", output=""}
   ]
   properties {}
  }
  MaterialExpressionCollectionParameter_0 {
   class=MaterialExpressionCollectionParameter
   inputs []
   properties {
    collection=&FortniteMaterialParameters
    parameterName=EnvironmentalEmissiveMult
   }
  }
  MaterialExpressionComponentMask_1 {class=MaterialExpressionComponentMask, inputs [None {from=MaterialExpressionActorPositionWS_0}], properties {g=true, r=true}}
  MaterialExpressionComponentMask_3 {class=MaterialExpressionComponentMask, inputs [None {from=MaterialExpressionVectorParameter_1, output=RGB}], properties {g=true, r=true}}
  MaterialExpressionComponentMask_4 {class=MaterialExpressionComponentMask, inputs [None {from=MaterialExpressionTextureCoordinate_2}], properties {g=true}}
  MaterialExpressionComponentMask_5 {class=MaterialExpressionComponentMask, inputs [None {from=MaterialExpressionAdd_7}], properties {g=true, r=true}}
  MaterialExpressionComponentMask_6 {class=MaterialExpressionComponentMask, inputs [None {from=MaterialExpressionAdd_7}], properties {b=true}}
  MaterialExpressionComponentMask_7 {class=MaterialExpressionComponentMask, inputs [None {from=MaterialExpressionVertexNormalWS_1}], properties {r=true}}
  MaterialExpressionConstant3Vector_0 {class=MaterialExpressionConstant3Vector, properties {constant=<rgba 0 0 1 0>}}
  MaterialExpressionConstant3Vector_1 {class=MaterialExpressionConstant3Vector}
  MaterialExpressionConstant3Vector_2 {class=MaterialExpressionConstant3Vector, properties {constant=<rgba 0 0 1 0>}}
  MaterialExpressionConstant3Vector_4 {class=MaterialExpressionConstant3Vector, properties {constant=<rgba 0 0 1 0>}}
  MaterialExpressionConstant3Vector_5 {class=MaterialExpressionConstant3Vector, properties {constant=<rgba 1 1 1 0>}}
  MaterialExpressionConstantBiasScale_2 {class=MaterialExpressionConstantBiasScale, inputs [None {from=MaterialExpressionTextureCoordinate_1}], properties {bias=-0.5, scale=2}}
  MaterialExpressionConstant_1 {class=MaterialExpressionConstant, properties {r=1}}
  MaterialExpressionConstant_10 {class=MaterialExpressionConstant, properties {r=1}}
  MaterialExpressionConstant_2 {class=MaterialExpressionConstant, properties {r=-1}}
  MaterialExpressionConstant_3 {class=MaterialExpressionConstant}
  MaterialExpressionConstant_4 {class=MaterialExpressionConstant}
  MaterialExpressionConstant_5 {class=MaterialExpressionConstant, properties {r=1}}
  MaterialExpressionConvert_0 {
   class=MaterialExpressionConvert
   inputs [None {from=MaterialExpressionMultiply_32}]
   properties {
    nodeName=Break Float3
    convertInputs [{
      type=Vector3
      expressionInput {
       expression=&self:MaterialExpressionMultiply_32
       inputName=None
       mask=0
       maskA=0
       maskB=0
       maskG=0
       maskR=0
       outputIndex=0
      }
     }]
    convertOutputs [{type=Scalar}, {type=Scalar}, {type=Scalar}]
   }
  }
  MaterialExpressionConvert_1 {
   class=MaterialExpressionConvert
   inputs [
    None {from=MaterialExpressionMaterialFunctionCall_20, output=sRGB}
    None {from=MaterialExpressionMaterialFunctionCall_21, output=sRGB}
    None {from=MaterialExpressionMaterialFunctionCall_22, output=sRGB}
   ]
   properties {
    nodeName=Make Vector3
    convertInputs [
     {
      type=Scalar
      expressionInput {
       expression=&self:MaterialExpressionMaterialFunctionCall_20
       inputName=None
       mask=0
       maskA=0
       maskB=0
       maskG=0
       maskR=0
       outputIndex=0
      }
     }
     {
      type=Scalar
      expressionInput {
       expression=&self:MaterialExpressionMaterialFunctionCall_21
       inputName=None
       mask=0
       maskA=0
       maskB=0
       maskG=0
       maskR=0
       outputIndex=0
      }
     }
     {
      type=Scalar
      expressionInput {
       expression=&self:MaterialExpressionMaterialFunctionCall_22
       inputName=None
       mask=0
       maskA=0
       maskB=0
       maskG=0
       maskR=0
       outputIndex=0
      }
     }
    ]
    convertOutputs [{type=Vector3}]
   }
  }
  MaterialExpressionConvert_2 {
   class=MaterialExpressionConvert
   inputs [None {from=MaterialExpressionMultiply_31}]
   properties {
    nodeName=Break Float3
    convertInputs [{
      type=Vector3
      expressionInput {
       expression=&self:MaterialExpressionMultiply_31
       inputName=None
       mask=0
       maskA=0
       maskB=0
       maskG=0
       maskR=0
       outputIndex=0
      }
     }]
    convertOutputs [{type=Scalar}, {type=Scalar}, {type=Scalar}]
   }
  }
  MaterialExpressionConvert_3 {
   class=MaterialExpressionConvert
   inputs [
    None {from=MaterialExpressionMaterialFunctionCall_23, output=sRGB}
    None {from=MaterialExpressionMaterialFunctionCall_24, output=sRGB}
    None {from=MaterialExpressionMaterialFunctionCall_25, output=sRGB}
   ]
   properties {
    nodeName=Make Vector3
    convertInputs [
     {
      type=Scalar
      expressionInput {
       expression=&self:MaterialExpressionMaterialFunctionCall_23
       inputName=None
       mask=0
       maskA=0
       maskB=0
       maskG=0
       maskR=0
       outputIndex=0
      }
     }
     {
      type=Scalar
      expressionInput {
       expression=&self:MaterialExpressionMaterialFunctionCall_24
       inputName=None
       mask=0
       maskA=0
       maskB=0
       maskG=0
       maskR=0
       outputIndex=0
      }
     }
     {
      type=Scalar
      expressionInput {
       expression=&self:MaterialExpressionMaterialFunctionCall_25
       inputName=None
       mask=0
       maskA=0
       maskB=0
       maskG=0
       maskR=0
       outputIndex=0
      }
     }
    ]
    convertOutputs [{type=Vector3}]
   }
  }
  MaterialExpressionCrossProduct_0 {
   class=MaterialExpressionCrossProduct
   inputs [
    A {from=MaterialExpressionDDX_0, output=""}
    B {from=MaterialExpressionDDY_0, output=""}
   ]
   properties {}
  }
  MaterialExpressionCrossProduct_12 {
   class=MaterialExpressionCrossProduct
   inputs [
    A {from=MaterialExpressionDDY_12, output=""}
    B {from=MaterialExpressionReroute_12, output=""}
   ]
   properties {desc=dp2perp}
  }
  MaterialExpressionCrossProduct_13 {
   class=MaterialExpressionCrossProduct
   inputs [
    A {from=MaterialExpressionReroute_12, output=""}
    B {from=MaterialExpressionDDX_13, output=""}
   ]
   properties {desc=dp1perp}
  }
  MaterialExpressionDDX_0 {class=MaterialExpressionDDX, inputs [Value {from=MaterialExpressionWorldPosition_1, output=XYZ}], properties {}}
  MaterialExpressionDDX_12 {class=MaterialExpressionDDX, inputs [Value {from=MaterialExpressionReroute_3, output=""}], properties {desc=duv1}}
  MaterialExpressionDDX_13 {class=MaterialExpressionDDX, inputs [Value {from=MaterialExpressionReroute_11, output=""}], properties {desc=dp1}}
  MaterialExpressionDDY_0 {class=MaterialExpressionDDY, inputs [Value {from=MaterialExpressionWorldPosition_1, output=XYZ}], properties {}}
  MaterialExpressionDDY_12 {class=MaterialExpressionDDY, inputs [Value {from=MaterialExpressionReroute_11, output=""}], properties {desc=dp2}}
  MaterialExpressionDDY_13 {class=MaterialExpressionDDY, inputs [Value {from=MaterialExpressionReroute_3, output=""}], properties {desc=duv2}}
  MaterialExpressionDesaturation_0 {
   class=MaterialExpressionDesaturation
   inputs [
    None {from=MaterialExpressionAdd_13, output=""}
    Fraction {from=MaterialExpressionConstant_10, output=""}
   ]
   properties {}
  }
  MaterialExpressionDivide_0 {
   class=MaterialExpressionDivide
   inputs [
    A {from=MaterialExpressionComponentMask_1}
    B {from=""}
   ]
   properties {constB=100}
  }
  MaterialExpressionDivide_1 {
   class=MaterialExpressionDivide
   inputs [
    A {from=MaterialExpressionTransform_1}
    B {from=""}
   ]
  }
  MaterialExpressionDivide_5 {
   class=MaterialExpressionDivide
   inputs [
    A {from=""}
    B {from=MaterialExpressionSquareRoot_5}
   ]
  }
  MaterialExpressionDotProduct_1 {
   class=MaterialExpressionDotProduct
   inputs [
    A {from=MaterialExpressionConstant3Vector_2, output=""}
    B {from=MaterialExpressionSceneTexture_1, output=Color}
   ]
   properties {}
  }
  MaterialExpressionDotProduct_12 {
   class=MaterialExpressionDotProduct
   inputs [
    A {from=MaterialExpressionAdd_6, output=""}
    B {from=MaterialExpressionAdd_6, output=""}
   ]
   properties {}
  }
  MaterialExpressionDotProduct_13 {
   class=MaterialExpressionDotProduct
   inputs [
    A {from=MaterialExpressionAdd_11, output=""}
    B {from=MaterialExpressionAdd_11, output=""}
   ]
   properties {}
  }
  MaterialExpressionDotProduct_2 {
   class=MaterialExpressionDotProduct
   inputs [
    A {from=MaterialExpressionNormalize_4, output=""}
    B {from=MaterialExpressionNormalize_1, output=""}
   ]
   properties {}
  }
  MaterialExpressionDotProduct_4 {
   class=MaterialExpressionDotProduct
   inputs [
    A {from=MaterialExpressionVertexNormalWS_2, output=""}
    B {from=MaterialExpressionSkyAtmosphereLightDirection_0, output=""}
   ]
   properties {}
  }
  MaterialExpressionGetMaterialAttributes_1 {
   class=MaterialExpressionGetMaterialAttributes
   inputs [None {from=MaterialExpressionMaterialFunctionCall_6, output=Result}]
   properties {attributeGetTypes [#1, #2]}
  }
  MaterialExpressionInverseLinearInterpolate_2 {
   class=MaterialExpressionInverseLinearInterpolate
   inputs [
    A {from="", output=""}
    B {from="", output=""}
    Value {from=MaterialExpressionReroute_0, output=""}
   ]
   properties {constA=1, constB=0}
  }
  MaterialExpressionInverseLinearInterpolate_3 {
   class=MaterialExpressionInverseLinearInterpolate
   inputs [
    A {from=MaterialExpressionLinearInterpolate_2, output=""}
    B {from=MaterialExpressionLinearInterpolate_5, output=""}
    Value {from=MaterialExpressionReroute_0, output=""}
   ]
   properties {}
  }
  MaterialExpressionInverseLinearInterpolate_4 {
   class=MaterialExpressionInverseLinearInterpolate
   inputs [
    A {from="", output=""}
    B {from="", output=""}
    Value {from=MaterialExpressionComponentMask_4, output=""}
   ]
   properties {constA=0.4, constB=0.6}
  }
  MaterialExpressionInverseLinearInterpolate_5 {
   class=MaterialExpressionInverseLinearInterpolate
   inputs [
    A {from="", output=""}
    B {from=MaterialExpressionScalarParameter_18, output=""}
    Value {from=MaterialExpressionMultiply_15, output=""}
   ]
   properties {constA=0.5}
  }
  MaterialExpressionLength_0 {class=MaterialExpressionLength, inputs [None {from=MaterialExpressionConstantBiasScale_2, output=""}], properties {}}
  MaterialExpressionLinearInterpolate_0 {
   class=MaterialExpressionLinearInterpolate
   inputs [
    A {from=MaterialExpressionPower_5}
    B {from=""}
    Alpha {from=MaterialExpressionTextureSample_4, output=R}
   ]
  }
  MaterialExpressionLinearInterpolate_1 {
   class=MaterialExpressionLinearInterpolate
   inputs [
    A {from=MaterialExpressionConstant3Vector_1}
    B {from=MaterialExpressionVectorParameter_2, output=RGB}
    Alpha {from=MaterialExpressionNamedRerouteUsage_5}
   ]
  }
  MaterialExpressionLinearInterpolate_2 {
   class=MaterialExpressionLinearInterpolate
   inputs [
    A {from=MaterialExpressionScalarParameter_15}
    B {from=MaterialExpressionMultiply_1}
    Alpha {from=MaterialExpressionScalarParameter_4}
   ]
  }
  MaterialExpressionLinearInterpolate_3 {
   class=MaterialExpressionLinearInterpolate
   inputs [
    A {from=MaterialExpressionMaterialFunctionCall_10, output=Result}
    B {from=MaterialExpressionConstant3Vector_4}
    Alpha {from=MaterialExpressionSaturate_8}
   ]
  }
  MaterialExpressionLinearInterpolate_4 {
   class=MaterialExpressionLinearInterpolate
   inputs [
    A {from=MaterialExpressionLinearInterpolate_3}
    B {from=MaterialExpressionConstant3Vector_4}
    Alpha {from=MaterialExpressionCeil_0}
   ]
  }
  MaterialExpressionLinearInterpolate_5 {
   class=MaterialExpressionLinearInterpolate
   inputs [
    A {from=MaterialExpressionScalarParameter_16}
    B {from=MaterialExpressionMultiply_13}
    Alpha {from=MaterialExpressionScalarParameter_4}
   ]
  }
  MaterialExpressionLinearInterpolate_6 {
   class=MaterialExpressionLinearInterpolate
   inputs [
    A {from=""}
    B {from=MaterialExpressionInverseLinearInterpolate_4}
    Alpha {from=MaterialExpressionScalarParameter_20}
   ]
   properties {constA=1}
  }
  MaterialExpressionLinearInterpolate_7 {
   class=MaterialExpressionLinearInterpolate
   inputs [
    A {from=""}
    B {from=MaterialExpressionCollectionParameter_0}
    Alpha {from=MaterialExpressionSaturate_17}
   ]
   properties {constA=1, constB=0}
  }
  MaterialExpressionLinearInterpolate_8 {
   class=MaterialExpressionLinearInterpolate
   inputs [
    A {from=MaterialExpressionAdd_13}
    B {from=MaterialExpressionMultiply_34}
    Alpha {from=MaterialExpressionSaturate_18}
   ]
  }
  MaterialExpressionMakeMaterialAttributes_0 {
   class=MaterialExpressionMakeMaterialAttributes
   inputs [
    BaseColor {from=MaterialExpressionLinearInterpolate_1, output=""}
    Metallic {from=MaterialExpressionConstant_3, output=""}
    Specular {from=MaterialExpressionConstant_4, output=""}
    Roughness {from=MaterialExpressionConstant_5, output=""}
    Anisotropy {from="", output=""}
    EmissiveColor {from=MaterialExpressionLinearInterpolate_1, output=""}
    Opacity {from=MaterialExpressionNamedRerouteUsage_1, output=""}
    OpacityMask {from=MaterialExpressionNamedRerouteUsage_1, output=""}
    Normal {from="", output=""}
    Tangent {from="", output=""}
    WorldPositionOffset {from="", output=""}
    SubsurfaceColor {from="", output=""}
    ClearCoat {from="", output=""}
    ClearCoatRoughness {from="", output=""}
    AmbientOcclusion {from="", output=""}
    Refraction {from="", output=""}
    CustomizedUVs_0 {from="", output=""}
    CustomizedUVs_1 {from="", output=""}
    CustomizedUVs_2 {from="", output=""}
    CustomizedUVs_3 {from="", output=""}
    CustomizedUVs_4 {from="", output=""}
    CustomizedUVs_5 {from="", output=""}
    CustomizedUVs_6 {from="", output=""}
    CustomizedUVs_7 {from="", output=""}
    PixelDepthOffset {from="", output=""}
    ShadingModel {from="", output=""}
    Displacement {from="", output=""}
   ]
   properties {}
  }
  MaterialExpressionMakeMaterialAttributes_1 {
   class=MaterialExpressionMakeMaterialAttributes
   inputs [
    BaseColor {from=MaterialExpressionBreakMaterialAttributes_0, output=BaseColor}
    Metallic {from=MaterialExpressionBreakMaterialAttributes_0, output=BaseColor}
    Specular {from=MaterialExpressionBreakMaterialAttributes_0, output=BaseColor}
    Roughness {from=MaterialExpressionBreakMaterialAttributes_0, output=BaseColor}
    Anisotropy {from="", output=""}
    EmissiveColor {from=MaterialExpressionBreakMaterialAttributes_0, output=BaseColor}
    Opacity {from=MaterialExpressionNamedRerouteUsage_8, output=""}
    OpacityMask {from=MaterialExpressionNamedRerouteUsage_8, output=""}
    Normal {from="", output=""}
    Tangent {from="", output=""}
    WorldPositionOffset {from="", output=""}
    SubsurfaceColor {from="", output=""}
    ClearCoat {from="", output=""}
    ClearCoatRoughness {from="", output=""}
    AmbientOcclusion {from="", output=""}
    Refraction {from="", output=""}
    CustomizedUVs_0 {from="", output=""}
    CustomizedUVs_1 {from="", output=""}
    CustomizedUVs_2 {from="", output=""}
    CustomizedUVs_3 {from="", output=""}
    CustomizedUVs_4 {from="", output=""}
    CustomizedUVs_5 {from="", output=""}
    CustomizedUVs_6 {from="", output=""}
    CustomizedUVs_7 {from="", output=""}
    PixelDepthOffset {from="", output=""}
    ShadingModel {from="", output=""}
    Displacement {from="", output=""}
   ]
   properties {}
  }
  MaterialExpressionMaterialFunctionCall_0 {
   class=MaterialExpressionMaterialFunctionCall
   inputs [
    "Xb1,PS4" {from=MaterialExpressionSaturate_12}
    "PC-High&&Epic,XSX,PS5" {from=MaterialExpressionSaturate_12}
    "PC-Perf,Android" {from=MaterialExpressionConstant_1}
    Switch {from=MaterialExpressionConstant_1}
    "PC-Low(NotPerf)&&Medium" {from=MaterialExpressionConstant_1}
   ]
   properties {materialFunction=&PlatformSwitch_1D}
  }
  MaterialExpressionMaterialFunctionCall_10 {
   class=MaterialExpressionMaterialFunctionCall
   inputs [
    X {from=MaterialExpressionTextureSample_7, output=G}
    Y {from=MaterialExpressionMultiply_18}
    Z {from=MaterialExpressionTextureSample_7, output=G}
   ]
   properties {materialFunction=&MakeFloat3}
  }
  MaterialExpressionMaterialFunctionCall_12 {
   class=MaterialExpressionMaterialFunctionCall
   inputs [
    SceneDepth {from=""}
    Exclude WPO Offsets {from=""}
   ]
   properties {materialFunction=&WorldPositionBehindTranslucency}
  }
  MaterialExpressionMaterialFunctionCall_15 {
   class=MaterialExpressionMaterialFunctionCall
   inputs [Float2 {from=MaterialExpressionDDX_12}]
   properties {materialFunction=&BreakOutFloat2Components}
  }
  MaterialExpressionMaterialFunctionCall_16 {
   class=MaterialExpressionMaterialFunctionCall
   inputs [Float2 {from=MaterialExpressionDDY_13}]
   properties {materialFunction=&BreakOutFloat2Components}
  }
  MaterialExpressionMaterialFunctionCall_19 {
   class=MaterialExpressionMaterialFunctionCall
   inputs [
    Color_Morning {from=MaterialExpressionVectorParameter_4, output=RGB}
    Color_Day {from=MaterialExpressionVectorParameter_3, output=RGB}
    Color_Evening {from=MaterialExpressionVectorParameter_5, output=RGB}
    Color_Night {from=MaterialExpressionVectorParameter_6, output=RGB}
   ]
   properties {materialFunction=&Color_Time_of_Day}
  }
  MaterialExpressionMaterialFunctionCall_2 {
   class=MaterialExpressionMaterialFunctionCall
   inputs [
    "Xb1,PS4" {from=MaterialExpressionSceneTexture_0, output=Color}
    "PC-High&&Epic,XSX,PS5" {from=MaterialExpressionSceneTexture_0, output=Color}
    "PC-Perf,Android" {from=MaterialExpressionConstant3Vector_0}
    Switch {from=MaterialExpressionConstant3Vector_0}
    "PC-Low(NotPerf)&&Medium" {from=MaterialExpressionSceneTexture_0, output=Color}
   ]
   properties {materialFunction=/Game/Effects/Niagara/MasterMaterials/Functions/Common/PlatformSwitch_3D}
  }
  MaterialExpressionMaterialFunctionCall_20 {
   class=MaterialExpressionMaterialFunctionCall
   inputs [Linear {from=MaterialExpressionConvert_0}]
   properties {materialFunction=/Engine/Functions/Engine_MaterialFunctions01/ImageAdjustment/LinearTosRGB}
  }
  MaterialExpressionMaterialFunctionCall_21 {
   class=MaterialExpressionMaterialFunctionCall
   inputs [Linear {from=MaterialExpressionConvert_0}]
   properties {materialFunction=&LinearTosRGB}
  }
  MaterialExpressionMaterialFunctionCall_22 {
   class=MaterialExpressionMaterialFunctionCall
   inputs [Linear {from=MaterialExpressionConvert_0}]
   properties {materialFunction=&LinearTosRGB}
  }
  MaterialExpressionMaterialFunctionCall_23 {
   class=MaterialExpressionMaterialFunctionCall
   inputs [Linear {from=MaterialExpressionConvert_2}]
   properties {materialFunction=&LinearTosRGB}
  }
  MaterialExpressionMaterialFunctionCall_24 {
   class=MaterialExpressionMaterialFunctionCall
   inputs [Linear {from=MaterialExpressionConvert_2}]
   properties {materialFunction=&LinearTosRGB}
  }
  MaterialExpressionMaterialFunctionCall_25 {
   class=MaterialExpressionMaterialFunctionCall
   inputs [Linear {from=MaterialExpressionConvert_2}]
   properties {materialFunction=&LinearTosRGB}
  }
  MaterialExpressionMaterialFunctionCall_3 {
   class=MaterialExpressionMaterialFunctionCall
   inputs [
    VectorToTransform {from=MaterialExpressionMaterialFunctionCall_2, output=Result}
    BasisX {from=MaterialExpressionMultiply_41}
    BasisY {from=MaterialExpressionMultiply_42}
    BasisZ {from=MaterialExpressionReroute_4}
    BasisPosition {from=""}
   ]
   properties {materialFunction=&Transform3x3Matrix}
  }
  MaterialExpressionMaterialFunctionCall_6 {
   class=MaterialExpressionMaterialFunctionCall
   inputs [
    In {from=MaterialExpressionMakeMaterialAttributes_0}
    Banding Mask {from=MaterialExpressionNamedRerouteUsage_6}
    NormalWS {from=MaterialExpressionNamedRerouteUsage_7}
   ]
   properties {materialFunction=/SonicSpearGameplay/Assets/Materials/MF_ColorBanding}
  }
  MaterialExpressionMaterialFunctionCall_7 {
   class=MaterialExpressionMaterialFunctionCall
   inputs [
    BaseNormal {from=MaterialExpressionNamedRerouteUsage_2}
    AdditionalNormal {from=MaterialExpressionNormalize_0}
   ]
   properties {materialFunction=&BlendAngleCorrectedNormals}
  }
  MaterialExpressionMaterialFunctionCall_8 {
   class=MaterialExpressionMaterialFunctionCall
   inputs [
    SceneDepth {from=""}
    Exclude WPO Offsets {from=""}
   ]
   properties {materialFunction=&WorldPositionBehindTranslucency}
  }
  MaterialExpressionMaterialProxyReplace_0 {
   class=MaterialExpressionMaterialProxyReplace
   inputs [
    Realtime {from=MaterialExpressionStaticSwitch_0, output=""}
    MaterialProxy {from=MaterialExpressionReroute_6, output=""}
   ]
   properties {}
  }
  MaterialExpressionMax_1 {
   class=MaterialExpressionMax
   inputs [
    A {from=""}
    B {from=MaterialExpressionNamedRerouteUsage_4}
   ]
  }
  MaterialExpressionMax_2 {
   class=MaterialExpressionMax
   inputs [
    A {from=MaterialExpressionOneMinus_0}
    B {from=""}
   ]
   properties {constB=0.01}
  }
  MaterialExpressionMax_3 {
   class=MaterialExpressionMax
   inputs [
    A {from=""}
    B {from=MaterialExpressionSubtract_1}
   ]
  }
  MaterialExpressionMax_4 {
   class=MaterialExpressionMax
   inputs [
    A {from=""}
    B {from=MaterialExpressionAbs_0}
   ]
  }
  MaterialExpressionMax_5 {
   class=MaterialExpressionMax
   inputs [
    A {from=""}
    B {from=MaterialExpressionReroute_2}
   ]
  }
  MaterialExpressionMax_8 {
   class=MaterialExpressionMax
   inputs [
    A {from=MaterialExpressionDotProduct_12}
    B {from=MaterialExpressionDotProduct_13}
   ]
  }
  MaterialExpressionMultiply_0 {
   class=MaterialExpressionMultiply
   inputs [
    A {from=MaterialExpressionSubtract_0}
    B {from=MaterialExpressionScalarParameter_3}
   ]
   properties {desc=How fast it shrinks}
  }
  MaterialExpressionMultiply_1 {
   class=MaterialExpressionMultiply
   inputs [
    A {from=MaterialExpressionMax_2}
    B {from=MaterialExpressionScalarParameter_15}
   ]
  }
  MaterialExpressionMultiply_11 {
   class=MaterialExpressionMultiply
   inputs [
    A {from=MaterialExpressionTextureCoordinate_1}
    B {from=MaterialExpressionComponentMask_3}
   ]
  }
  MaterialExpressionMultiply_12 {
   class=MaterialExpressionMultiply
   inputs [
    A {from=MaterialExpressionInverseLinearInterpolate_3}
    B {from=MaterialExpressionLinearInterpolate_0}
   ]
  }
  MaterialExpressionMultiply_13 {
   class=MaterialExpressionMultiply
   inputs [
    A {from=MaterialExpressionMax_2}
    B {from=MaterialExpressionScalarParameter_16}
   ]
  }
  MaterialExpressionMultiply_15 {
   class=MaterialExpressionMultiply
   inputs [
    A {from=MaterialExpressionMultiply_12}
    B {from=MaterialExpressionSaturate_6}
   ]
  }
  MaterialExpressionMultiply_17 {
   class=MaterialExpressionMultiply
   inputs [
    A {from=MaterialExpressionInverseLinearInterpolate_5}
    B {from=""}
   ]
   properties {constB=-1}
  }
  MaterialExpressionMultiply_18 {
   class=MaterialExpressionMultiply
   inputs [
    A {from=MaterialExpressionTextureSample_7, output=R}
    B {from=""}
   ]
   properties {constB=-1}
  }
  MaterialExpressionMultiply_19 {
   class=MaterialExpressionMultiply
   inputs [
    A {from=MaterialExpressionObjectOrientation_0}
    B {from=MaterialExpressionConstant_2}
   ]
  }
  MaterialExpressionMultiply_2 {
   class=MaterialExpressionMultiply
   inputs [
    A {from=MaterialExpressionLinearInterpolate_1}
    B {from=""}
   ]
   properties {constB=0.2}
  }
  MaterialExpressionMultiply_20 {
   class=MaterialExpressionMultiply
   inputs [
    A {from=MaterialExpressionMax_1}
    B {from=MaterialExpressionMaterialFunctionCall_0, output=Result}
   ]
  }
  MaterialExpressionMultiply_21 {
   class=MaterialExpressionMultiply
   inputs [
    A {from=MaterialExpressionPower_0}
    B {from=MaterialExpressionScalarParameter_22}
   ]
  }
  MaterialExpressionMultiply_22 {
   class=MaterialExpressionMultiply
   inputs [
    A {from=MaterialExpressionMax_4}
    B {from=MaterialExpressionCeil_1}
   ]
  }
  MaterialExpressionMultiply_23 {
   class=MaterialExpressionMultiply
   inputs [
    A {from=MaterialExpressionNamedRerouteUsage_3}
    B {from=MaterialExpressionMultiply_20}
   ]
  }
  MaterialExpressionMultiply_25 {
   class=MaterialExpressionMultiply
   inputs [
    A {from=MaterialExpressionSaturate_14}
    B {from=""}
   ]
   properties {constB=0.5}
  }
  MaterialExpressionMultiply_26 {
   class=MaterialExpressionMultiply
   inputs [
    A {from=MaterialExpressionMultiply_28}
    B {from=MaterialExpressionNamedRerouteUsage_11}
   ]
   properties {constB=0.2}
  }
  MaterialExpressionMultiply_27 {
   class=MaterialExpressionMultiply
   inputs [
    A {from=MaterialExpressionAdd_12}
    B {from=MaterialExpressionScalarParameter_24}
   ]
   properties {constB=0.12}
  }
  MaterialExpressionMultiply_28 {
   class=MaterialExpressionMultiply
   inputs [
    A {from=MaterialExpressionClamp_0}
    B {from=MaterialExpressionAdd_9}
   ]
  }
  MaterialExpressionMultiply_29 {
   class=MaterialExpressionMultiply
   inputs [
    A {from=MaterialExpressionAdd_10}
    B {from=""}
   ]
   properties {constB=0.5}
  }
  MaterialExpressionMultiply_30 {
   class=MaterialExpressionMultiply
   inputs [
    A {from=MaterialExpressionMultiply_28}
    B {from=MaterialExpressionLinearInterpolate_7}
   ]
  }
  MaterialExpressionMultiply_31 {
   class=MaterialExpressionMultiply
   inputs [
    A {from=MaterialExpressionNamedRerouteUsage_12}
    B {from=MaterialExpressionNamedRerouteUsage_10}
   ]
   properties {constB=0.2}
  }
  MaterialExpressionMultiply_32 {
   class=MaterialExpressionMultiply
   inputs [
    A {from=MaterialExpressionNamedRerouteUsage_12}
    B {from=""}
   ]
   properties {constB=0.6}
  }
  MaterialExpressionMultiply_33 {
   class=MaterialExpressionMultiply
   inputs [
    A {from=MaterialExpressionMultiply_30}
    B {from=""}
   ]
   properties {constB=0.75}
  }
  MaterialExpressionMultiply_34 {
   class=MaterialExpressionMultiply
   inputs [
    A {from=MaterialExpressionAdd_13}
    B {from=MaterialExpressionScalarParameter_26}
   ]
   properties {constB=0.2}
  }
  MaterialExpressionMultiply_37 {
   class=MaterialExpressionMultiply
   inputs [
    A {from=MaterialExpressionCrossProduct_12}
    B {from=MaterialExpressionMaterialFunctionCall_15, output=R}
   ]
  }
  MaterialExpressionMultiply_38 {
   class=MaterialExpressionMultiply
   inputs [
    A {from=MaterialExpressionCrossProduct_13}
    B {from=MaterialExpressionMaterialFunctionCall_16, output=R}
   ]
  }
  MaterialExpressionMultiply_39 {
   class=MaterialExpressionMultiply
   inputs [
    A {from=MaterialExpressionCrossProduct_12}
    B {from=MaterialExpressionMaterialFunctionCall_15, output=G}
   ]
  }
  MaterialExpressionMultiply_40 {
   class=MaterialExpressionMultiply
   inputs [
    A {from=MaterialExpressionCrossProduct_13}
    B {from=MaterialExpressionMaterialFunctionCall_16, output=G}
   ]
  }
  MaterialExpressionMultiply_41 {
   class=MaterialExpressionMultiply
   inputs [
    A {from=MaterialExpressionAdd_6}
    B {from=MaterialExpressionDivide_5}
   ]
   properties {desc=Adjusted T}
  }
  MaterialExpressionMultiply_42 {
   class=MaterialExpressionMultiply
   inputs [
    A {from=MaterialExpressionDivide_5}
    B {from=MaterialExpressionDotProduct_13}
   ]
   properties {desc=Adjusted T}
  }
  MaterialExpressionNamedRerouteDeclaration_0 {
   class=MaterialExpressionNamedRerouteDeclaration
   inputs [None {from=MaterialExpressionMaterialFunctionCall_2, output=Result}]
   properties {name=NormalWS, nodeColor=<rgba 0.4766 1 0 1>}
  }
  MaterialExpressionNamedRerouteDeclaration_1 {
   class=MaterialExpressionNamedRerouteDeclaration
   inputs [None {from=MaterialExpressionMultiply_20}]
   properties {name=Opacity, nodeColor=<rgba 0 0.25 1 1>}
  }
  MaterialExpressionNamedRerouteDeclaration_2 {
   class=MaterialExpressionNamedRerouteDeclaration
   inputs [None {from=MaterialExpressionSaturate_10}]
   properties {name=AngleFade, nodeColor=<rgba 0.5703 1 0 1>}
  }
  MaterialExpressionNamedRerouteDeclaration_3 {
   class=MaterialExpressionNamedRerouteDeclaration
   inputs [None {from=MaterialExpressionRound_0}]
   properties {name=BandingMask, nodeColor=<rgba 1 0 0.8906 1>}
  }
  MaterialExpressionNamedRerouteDeclaration_4 {
   class=MaterialExpressionNamedRerouteDeclaration
   inputs [None {from=MaterialExpressionMaterialFunctionCall_7, output=Result}]
   properties {name=NormalsFinal, nodeColor=<rgba 0 1 0.7891 1>}
  }
  MaterialExpressionNamedRerouteDeclaration_7 {
   class=MaterialExpressionNamedRerouteDeclaration
   inputs [None {from=MaterialExpressionGetMaterialAttributes_1, output=Emissive Color}]
   properties {name=Emission, nodeColor=<rgba 1 0 0.3516 1>}
  }
  MaterialExpressionNamedRerouteDeclaration_8 {
   class=MaterialExpressionNamedRerouteDeclaration
   inputs [None {from=MaterialExpressionMaterialFunctionCall_19, output=Result}]
   properties {name=Unlit Emission, nodeColor=<rgba 1 0.1172 0 1>}
  }
  MaterialExpressionNamedRerouteDeclaration_9 {
   class=MaterialExpressionNamedRerouteDeclaration
   inputs [None {from=MaterialExpressionGetMaterialAttributes_1, output=Base Color}]
   properties {name=Base Color, nodeColor=<rgba 0 0.2031 1 1>}
  }
  MaterialExpressionNamedRerouteUsage_1 {class=MaterialExpressionNamedRerouteUsage}
  MaterialExpressionNamedRerouteUsage_10 {class=MaterialExpressionNamedRerouteUsage}
  MaterialExpressionNamedRerouteUsage_11 {class=MaterialExpressionNamedRerouteUsage}
  MaterialExpressionNamedRerouteUsage_12 {class=MaterialExpressionNamedRerouteUsage}
  MaterialExpressionNamedRerouteUsage_13 {class=MaterialExpressionNamedRerouteUsage}
  MaterialExpressionNamedRerouteUsage_2 {class=MaterialExpressionNamedRerouteUsage}
  MaterialExpressionNamedRerouteUsage_3 {class=MaterialExpressionNamedRerouteUsage}
  MaterialExpressionNamedRerouteUsage_4 {class=MaterialExpressionNamedRerouteUsage}
  MaterialExpressionNamedRerouteUsage_5 {class=MaterialExpressionNamedRerouteUsage}
  MaterialExpressionNamedRerouteUsage_6 {class=MaterialExpressionNamedRerouteUsage}
  MaterialExpressionNamedRerouteUsage_7 {class=MaterialExpressionNamedRerouteUsage}
  MaterialExpressionNamedRerouteUsage_8 {class=MaterialExpressionNamedRerouteUsage}
  MaterialExpressionNamedRerouteUsage_9 {class=MaterialExpressionNamedRerouteUsage}
  MaterialExpressionNormalize_0 {class=MaterialExpressionNormalize, inputs [VectorInput {from=MaterialExpressionTransform_0, output=""}]}
  MaterialExpressionNormalize_1 {class=MaterialExpressionNormalize, inputs [VectorInput {from=MaterialExpressionCrossProduct_0, output=""}]}
  MaterialExpressionNormalize_4 {class=MaterialExpressionNormalize, inputs [VectorInput {from=MaterialExpressionMultiply_19, output=""}]}
  MaterialExpressionObjectOrientation_0 {class=MaterialExpressionObjectOrientation, inputs [], properties {}}
  MaterialExpressionObjectPositionWS_0 {class=MaterialExpressionObjectPositionWS, inputs [], properties {}}
  MaterialExpressionOneMinus_0 {class=MaterialExpressionOneMinus, inputs [None {from=MaterialExpressionSaturate_1}]}
  MaterialExpressionParticlePositionWS_0 {class=MaterialExpressionParticlePositionWS, inputs [], properties {}}
  MaterialExpressionPower_0 {
   class=MaterialExpressionPower
   inputs [
    Base {from=MaterialExpressionMax_5}
    Exp {from=MaterialExpressionScalarParameter_19}
   ]
  }
  MaterialExpressionPower_5 {
   class=MaterialExpressionPower
   inputs [
    Base {from=MaterialExpressionSaturate_5}
    Exp {from=""}
   ]
  }
  MaterialExpressionReroute_0 {class=MaterialExpressionReroute, inputs [None {from=MaterialExpressionLength_0}]}
  MaterialExpressionReroute_1 {class=MaterialExpressionReroute, inputs [None {from=MaterialExpressionMaterialProxyReplace_0}]}
  MaterialExpressionReroute_11 {class=MaterialExpressionReroute, inputs [None {from=MaterialExpressionWorldPosition_3, output=XYZ}]}
  MaterialExpressionReroute_12 {class=MaterialExpressionReroute, inputs [None {from=MaterialExpressionConstant3Vector_5}]}
  MaterialExpressionReroute_2 {class=MaterialExpressionReroute, inputs [None {from=MaterialExpressionMultiply_22}]}
  MaterialExpressionReroute_3 {class=MaterialExpressionReroute, inputs [None {from=MaterialExpressionTextureCoordinate_3}]}
  MaterialExpressionReroute_4 {class=MaterialExpressionReroute, inputs [None {from=MaterialExpressionReroute_5}]}
  MaterialExpressionReroute_5 {class=MaterialExpressionReroute, inputs [None {from=MaterialExpressionConstant3Vector_5}]}
  MaterialExpressionReroute_6 {class=MaterialExpressionReroute, inputs [None {from=MaterialExpressionSetMaterialAttributes_3}]}
  MaterialExpressionReroute_7 {class=MaterialExpressionReroute, inputs [None {from=MaterialExpressionMaterialFunctionCall_6, output=Result}]}
  MaterialExpressionRound_0 {class=MaterialExpressionRound, inputs [None {from=MaterialExpressionSaturate_4, output=""}], properties {}}
  MaterialExpressionRound_1 {class=MaterialExpressionRound, inputs [None {from=MaterialExpressionSaturate_11, output=""}], properties {}}
  MaterialExpressionSaturate_1 {class=MaterialExpressionSaturate, inputs [None {from=MaterialExpressionMultiply_0}]}
  MaterialExpressionSaturate_10 {class=MaterialExpressionSaturate, inputs [None {from=MaterialExpressionMultiply_21}]}
  MaterialExpressionSaturate_11 {class=MaterialExpressionSaturate, inputs [None {from=MaterialExpressionAdd_5}]}
  MaterialExpressionSaturate_12 {class=MaterialExpressionSaturate, inputs [None {from=MaterialExpressionRound_1}]}
  MaterialExpressionSaturate_13 {class=MaterialExpressionSaturate, inputs [None {from=MaterialExpressionComponentMask_5}]}
  MaterialExpressionSaturate_14 {class=MaterialExpressionSaturate, inputs [None {from=MaterialExpressionComponentMask_6}]}
  MaterialExpressionSaturate_15 {class=MaterialExpressionSaturate, inputs [None {from=MaterialExpressionSine_1}]}
  MaterialExpressionSaturate_16 {class=MaterialExpressionSaturate, inputs [None {from=MaterialExpressionComponentMask_6}]}
  MaterialExpressionSaturate_17 {class=MaterialExpressionSaturate, inputs [None {from=MaterialExpressionMultiply_29}]}
  MaterialExpressionSaturate_18 {class=MaterialExpressionSaturate, inputs [None {from=MaterialExpressionSubtract_5}]}
  MaterialExpressionSaturate_4 {class=MaterialExpressionSaturate, inputs [None {from=MaterialExpressionMultiply_15}]}
  MaterialExpressionSaturate_5 {class=MaterialExpressionSaturate, inputs [None {from=MaterialExpressionInverseLinearInterpolate_2}]}
  MaterialExpressionSaturate_6 {class=MaterialExpressionSaturate, inputs [None {from=MaterialExpressionLinearInterpolate_6}]}
  MaterialExpressionSaturate_8 {class=MaterialExpressionSaturate, inputs [None {from=MaterialExpressionInverseLinearInterpolate_5}]}
  MaterialExpressionSaturate_9 {class=MaterialExpressionSaturate, inputs [None {from=MaterialExpressionMultiply_17}]}
  MaterialExpressionScalarParameter_15 {class=MaterialExpressionScalarParameter, properties {defaultValue=1, parameterName=Frosting Distance Min}}
  MaterialExpressionScalarParameter_16 {class=MaterialExpressionScalarParameter, properties {defaultValue=0.85, parameterName=Frosting Distance Max}}
  MaterialExpressionScalarParameter_18 {class=MaterialExpressionScalarParameter, properties {defaultValue=3, parameterName=Frosting Normal Range}}
  MaterialExpressionScalarParameter_19 {
   class=MaterialExpressionScalarParameter
   properties {
    defaultValue=8
    desc=Adjust Angle fade falloff exponent
    group="05) Opacity---------------------------------------------------------------------------------------------------------------------------"
    parameterName=Angle Fade Exponent
    sortPriority=7
   }
  }
  MaterialExpressionScalarParameter_2 {class=MaterialExpressionScalarParameter, properties {parameterName=CurrentGameTime}}
  MaterialExpressionScalarParameter_20 {class=MaterialExpressionScalarParameter, properties {defaultValue=1, parameterName=ShapeAlpha}}
  MaterialExpressionScalarParameter_21 {class=MaterialExpressionScalarParameter, properties {defaultValue=-0.15, parameterName=SlopeAlphaOffset}}
  MaterialExpressionScalarParameter_22 {
   class=MaterialExpressionScalarParameter
   properties {
    defaultValue=10
    desc=Adjust Angle Fade amount
    group="05) Opacity---------------------------------------------------------------------------------------------------------------------------"
    parameterName=Angle Fade Multiplier
    sortPriority=6
   }
  }
  MaterialExpressionScalarParameter_23 {class=MaterialExpressionScalarParameter, properties {defaultValue=0.6, parameterName=Sun Supress}}
  MaterialExpressionScalarParameter_24 {class=MaterialExpressionScalarParameter, properties {defaultValue=0.2, parameterName=Corner Break}}
  MaterialExpressionScalarParameter_25 {class=MaterialExpressionScalarParameter, properties {defaultValue=0.025, parameterName=Base Color Add}}
  MaterialExpressionScalarParameter_26 {class=MaterialExpressionScalarParameter, properties {defaultValue=0.65, parameterName=Base Color Mult}}
  MaterialExpressionScalarParameter_27 {class=MaterialExpressionScalarParameter, properties {defaultValue=0.2, parameterName=Base Color Threshold}}
  MaterialExpressionScalarParameter_28 {class=MaterialExpressionScalarParameter, properties {defaultValue=0.05, parameterName=Base Color Min}}
  MaterialExpressionScalarParameter_3 {class=MaterialExpressionScalarParameter, properties {defaultValue=0.21, parameterName=WPO Speed}}
  MaterialExpressionScalarParameter_4 {class=MaterialExpressionScalarParameter, properties {parameterName=Shrinking Enabled}}
  MaterialExpressionSceneTexture_0 {class=MaterialExpressionSceneTexture, inputs [UVs {from="", output=""}], properties {sceneTextureId=PPI_WorldNormal}}
  MaterialExpressionSceneTexture_1 {class=MaterialExpressionSceneTexture, inputs [UVs {from="", output=""}], properties {sceneTextureId=PPI_WorldNormal}}
  MaterialExpressionSetMaterialAttributes_2 {
   class=MaterialExpressionSetMaterialAttributes
   inputs [
    MaterialAttributes {from=MaterialExpressionGetMaterialAttributes_1, output=MaterialAttributes}
    Base Color {from=MaterialExpressionMultiply_33, output=""}
    Emissive Color {from=MaterialExpressionAdd_8, output=""}
   ]
   properties {attributeSetTypes [#1, #2]}
  }
  MaterialExpressionSetMaterialAttributes_3 {
   class=MaterialExpressionSetMaterialAttributes
   inputs [
    MaterialAttributes {from=MaterialExpressionGetMaterialAttributes_1, output=MaterialAttributes}
    Base Color {from=MaterialExpressionMultiply_32, output=""}
    Emissive Color {from=MaterialExpressionMultiply_31, output=""}
   ]
   properties {attributeSetTypes [#1, #2]}
  }
  MaterialExpressionSine_1 {class=MaterialExpressionSine, inputs [None {from=MaterialExpressionMultiply_25}]}
  MaterialExpressionSkyAtmosphereLightDirection_0 {class=MaterialExpressionSkyAtmosphereLightDirection, inputs [], properties {}}
  MaterialExpressionSquareRoot_5 {class=MaterialExpressionSquareRoot, inputs [None {from=MaterialExpressionMax_8, output=""}], properties {}}
  MaterialExpressionStaticSwitch_0 {
   class=MaterialExpressionStaticSwitch
   inputs [
    True {from=MaterialExpressionBlendMaterialAttributes_0, output=""}
    False {from=MaterialExpressionReroute_6, output=""}
    Value {from="", output=""}
   ]
   properties {}
  }
  MaterialExpressionSubtract_0 {
   class=MaterialExpressionSubtract
   inputs [
    A {from=MaterialExpressionTime_0}
    B {from=MaterialExpressionScalarParameter_2}
   ]
  }
  MaterialExpressionSubtract_1 {
   class=MaterialExpressionSubtract
   inputs [
    A {from=MaterialExpressionDotProduct_2}
    B {from=""}
   ]
   properties {constB=0.05}
  }
  MaterialExpressionSubtract_3 {
   class=MaterialExpressionSubtract
   inputs [
    A {from=MaterialExpressionMaterialFunctionCall_8, output=Result}
    B {from=MaterialExpressionParticlePositionWS_0}
   ]
  }
  MaterialExpressionSubtract_4 {
   class=MaterialExpressionSubtract
   inputs [
    A {from=MaterialExpressionMaterialFunctionCall_12, output=Result}
    B {from=MaterialExpressionObjectPositionWS_0}
   ]
  }
  MaterialExpressionSubtract_5 {
   class=MaterialExpressionSubtract
   inputs [
    A {from=MaterialExpressionDesaturation_0}
    B {from=MaterialExpressionScalarParameter_27}
   ]
   properties {constB=0.45}
  }
  MaterialExpressionTextureCoordinate_1 {class=MaterialExpressionTextureCoordinate}
  MaterialExpressionTextureCoordinate_2 {class=MaterialExpressionTextureCoordinate}
  MaterialExpressionTextureCoordinate_3 {class=MaterialExpressionTextureCoordinate}
  MaterialExpressionTextureSample_4 {
   class=MaterialExpressionTextureSample
   inputs [
    UVs {from=MaterialExpressionAdd_0}
    Tex {from=""}
    Apply View MipBias {from=""}
   ]
   properties {samplerType=SAMPLERTYPE_Grayscale, texture=/Game/Global/Textures/Default/Noise/LowResBlurredNoiseNormalized}
  }
  MaterialExpressionTextureSample_7 {
   class=MaterialExpressionTextureSample
   inputs [
    UVs {from=MaterialExpressionAdd_0}
    Tex {from=""}
    Apply View MipBias {from=""}
   ]
   properties {samplerType=SAMPLERTYPE_Normal, texture=/SonicSpearGameplay/Assets/Textures/LowResBlurredNoiseNormal}
  }
  MaterialExpressionTime_0 {class=MaterialExpressionTime}
  MaterialExpressionTransform_0 {class=MaterialExpressionTransform, inputs [None {from=MaterialExpressionLinearInterpolate_4}]}
  MaterialExpressionTransform_1 {
   class=MaterialExpressionTransform
   inputs [None {from=MaterialExpressionSubtract_3}]
   properties {transformSourceType=TRANSFORMSOURCE_World, transformType=TRANSFORM_Instance}
  }
  MaterialExpressionVectorParameter_1 {class=MaterialExpressionVectorParameter, properties {parameterName=Frosting Noise Tiling, defaultValue=<rgba 3 2 0 1>}}
  MaterialExpressionVectorParameter_2 {class=MaterialExpressionVectorParameter, properties {parameterName=Frosting Color, defaultValue=<rgba 0.8031 0.0209 0.5188 0>}}
  MaterialExpressionVectorParameter_3 {class=MaterialExpressionVectorParameter, properties {group=FlatShading, parameterName=Test_Day, defaultValue=<rgba 0.276 0.276 0.276 1>}}
  MaterialExpressionVectorParameter_4 {
   class=MaterialExpressionVectorParameter
   properties {group=FlatShading, parameterName=Test_Morning, defaultValue=<rgba 0.2293 0.2167 0.2865 1>}
  }
  MaterialExpressionVectorParameter_5 {
   class=MaterialExpressionVectorParameter
   properties {group=FlatShading, parameterName=Test_Evening, defaultValue=<rgba 0.2969 0.2047 0.292 1>}
  }
  MaterialExpressionVectorParameter_6 {class=MaterialExpressionVectorParameter, properties {group=FlatShading, parameterName=Test_Night, defaultValue=<rgba 0.2109 0.2253 0.25 1>}}
  MaterialExpressionVertexNormalWS_1 {class=MaterialExpressionVertexNormalWS, inputs [], properties {}}
  MaterialExpressionVertexNormalWS_2 {class=MaterialExpressionVertexNormalWS, inputs [], properties {}}
  MaterialExpressionWorldPosition_1 {class=MaterialExpressionWorldPosition, inputs [], properties {worldPositionShaderOffset=WPT_CameraRelative}}
  MaterialExpressionWorldPosition_3 {class=MaterialExpressionWorldPosition, inputs [], properties {worldPositionShaderOffset=WPT_CameraRelative}}
 }
 properties {
  bAutomaticallySetUsageInEditor=false
  bUseMaterialAttributes=true
  blendMode=BLEND_Translucent
  materialDomain=MD_DeferredDecal
  usedShadingModels=Default Lit
 }
}
