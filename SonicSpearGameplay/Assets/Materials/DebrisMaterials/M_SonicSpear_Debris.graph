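# Lit debris material for the Sonic Spear, built for instanced and Niagara meshes, with static switches choosing frosting, sprinkle or curve-atlas color.
# DBuffer decals adjust specular and roughness before color banding and optional flat shading.
# When enabled through custom primitive data, world position offset shrinks each instance into its pivot over time since CurrentGameTime at randomized speed.
# A further offset snaps debris vertically to the Apollo landscape height virtual texture.
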Material /SonicSpearGameplay/Assets/Materials/DebrisMaterials/M_SonicSpear_Debris
graph {
 expressions {
  MaterialExpressionAdd_0 {
   class=MaterialExpressionAdd
   inputs [
    A {from=MaterialExpressionConstant_2}
    B {from=MaterialExpressionDivide_0}
   ]
  }
  MaterialExpressionAdd_1 {
   class=MaterialExpressionAdd
   inputs [
    A {from=MaterialExpressionConstant_1}
    B {from=MaterialExpressionComponentMask_0}
   ]
  }
  MaterialExpressionAdd_6 {
   class=MaterialExpressionAdd
   inputs [
    A {from=MaterialExpressionSubtract_7}
    B {from=MaterialExpressionSubtract_5}
   ]
  }
  MaterialExpressionComponentMask_0 {class=MaterialExpressionComponentMask, inputs [None {from=MaterialExpressionDBufferTexture_0, output=RGB}], properties {r=true}}
  MaterialExpressionComponentMask_1 {class=MaterialExpressionComponentMask, inputs [None {from=MaterialExpressionConstant3Vector_0}], properties {b=true}}
  MaterialExpressionComponentMask_2 {class=MaterialExpressionComponentMask, inputs [None {from=MaterialExpressionTransformPosition_1}], properties {b=true}}
  MaterialExpressionConstant3Vector_0 {class=MaterialExpressionConstant3Vector}
  MaterialExpressionConstant3Vector_1 {class=MaterialExpressionConstant3Vector}
  MaterialExpressionConstant_0 {class=MaterialExpressionConstant}
  MaterialExpressionConstant_1 {class=MaterialExpressionConstant}
  MaterialExpressionConstant_2 {class=MaterialExpressionConstant, properties {r=0.5}}
  MaterialExpressionConstant_3 {class=MaterialExpressionConstant, properties {r=1}}
  MaterialExpressionConstant_4 {class=MaterialExpressionConstant}
  MaterialExpressionConstant_5 {class=MaterialExpressionConstant}
  MaterialExpressionConvert_0 {
   class=MaterialExpressionConvert
   inputs [
    None {from=""}
    None {from=""}
    None {from=MaterialExpressionAdd_6}
   ]
   properties {
    nodeName=Make Vector3
    convertInputs [
     {type=Scalar, expressionInput {expression=null, inputName=None, mask=0, maskA=0, maskB=0, maskG=0, maskR=0, outputIndex=0}}
     {type=Scalar, expressionInput {expression=null, inputName=None, mask=0, maskA=0, maskB=0, maskG=0, maskR=0, outputIndex=0}}
     {
      type=Scalar
      expressionInput {
       expression=&self:MaterialExpressionAdd_6
       inputName=None
       mask=0
       maskA=0
       maskB=0
       maskG=0
       maskR=0
       outputIndex=0
      }
     }
    ]
    convertOutputs [{type=Vector3}]
   }
  }
  MaterialExpressionCurveAtlasRowParameter_3 {
   class=MaterialExpressionCurveAtlasRowParameter
   inputs [CurveTime {from=MaterialExpressionConstant_5, output=""}]
   properties {atlas=/SonicSpearGameplay/Assets/Curves/CA_SonicSpear, curve=/SonicSpearGameplay/Assets/Curves/C_SonicSpear_Pink, parameterName=ColorCurve}
  }
  MaterialExpressionDBufferTexture_0 {class=MaterialExpressionDBufferTexture, inputs [UVs {from="", output=""}], properties {dBufferTextureId=DBT_C}}
  MaterialExpressionDivide_0 {
   class=MaterialExpressionDivide
   inputs [
    A {from=MaterialExpressionComponentMask_0}
    B {from=""}
   ]
  }
  MaterialExpressionGetMaterialAttributes_0 {
   class=MaterialExpressionGetMaterialAttributes
   inputs [None {from=MaterialExpressionMaterialFunctionCall_2, output=Result}]
   properties {attributeGetTypes [#1]}
  }
  MaterialExpressionLinearInterpolate_0 {
   class=MaterialExpressionLinearInterpolate
   inputs [
    A {from=""}
    B {from=MaterialExpressionMultiply_1}
    Alpha {from=MaterialExpressionScalarParameter_2}
   ]
  }
  MaterialExpressionLinearInterpolate_1 {
   class=MaterialExpressionLinearInterpolate
   inputs [
    A {from=""}
    B {from=""}
    Alpha {from=MaterialExpressionPerInstanceRandom_0}
   ]
   properties {constA=0.8, constB=2}
  }
  MaterialExpressionMaterialFunctionCall_0 {
   class=MaterialExpressionMaterialFunctionCall
   inputs []
   properties {materialFunction=&ObjectPivotPoint}
  }
  MaterialExpressionMaterialFunctionCall_2 {
   class=MaterialExpressionMaterialFunctionCall
   inputs [
    Material In {from=MaterialExpressionMaterialFunctionCall_4, output=Result}
    UseFlatShading {from=MaterialExpressionStaticBoolParameter_1}
    IsWorldSpace {from=MaterialExpressionStaticBool_0}
   ]
   properties {materialFunction=/SonicSpearGameplay/Assets/Materials/MF_WildEstate_FlatShading}
  }
  MaterialExpressionMaterialFunctionCall_4 {
   class=MaterialExpressionMaterialFunctionCall
   inputs [
    In {from=MaterialExpressionSetMaterialAttributes_0}
    Banding Mask {from=MaterialExpressionStaticSwitchParameter_0}
    NormalWS {from=""}
   ]
   properties {materialFunction=/SonicSpearGameplay/Assets/Materials/MF_ColorBanding}
  }
  MaterialExpressionMultiply_0 {
   class=MaterialExpressionMultiply
   inputs [
    A {from=MaterialExpressionSubtract_0}
    B {from=MaterialExpressionMultiply_2}
   ]
  }
  MaterialExpressionMultiply_1 {
   class=MaterialExpressionMultiply
   inputs [
    A {from=MaterialExpressionSaturate_0}
    B {from=MaterialExpressionSubtract_1}
   ]
  }
  MaterialExpressionMultiply_2 {
   class=MaterialExpressionMultiply
   inputs [
    A {from=MaterialExpressionScalarParameter_1}
    B {from=MaterialExpressionLinearInterpolate_1}
   ]
  }
  MaterialExpressionPerInstanceRandom_0 {class=MaterialExpressionPerInstanceRandom, inputs [], properties {}}
  MaterialExpressionRuntimeVirtualTextureSample_0 {
   class=MaterialExpressionRuntimeVirtualTextureSample
   inputs [
    UVs {from="", output=""}
    World Position {from="", output=""}
   ]
   properties {materialType=WorldHeight, virtualTexture=/Game/Athena/Apollo/Maps/Landscape/VT/RVT_Apollo_Landscape_Height}
  }
  MaterialExpressionSaturate_0 {class=MaterialExpressionSaturate, inputs [None {from=MaterialExpressionMultiply_0}]}
  MaterialExpressionSaturate_1 {class=MaterialExpressionSaturate, inputs [None {from=MaterialExpressionAdd_0}]}
  MaterialExpressionScalarParameter_0 {class=MaterialExpressionScalarParameter, properties {bUseCustomPrimitiveData=true, parameterName=CurrentGameTime}}
  MaterialExpressionScalarParameter_1 {class=MaterialExpressionScalarParameter, properties {defaultValue=0.2, parameterName=WPO Speed}}
  MaterialExpressionScalarParameter_2 {class=MaterialExpressionScalarParameter, properties {bUseCustomPrimitiveData=true, parameterName=WPO Enabled, primitiveDataIndex=1}}
  MaterialExpressionSetMaterialAttributes_0 {
   class=MaterialExpressionSetMaterialAttributes
   inputs [
    MaterialAttributes {from="", output=""}
    Base Color {from=MaterialExpressionStaticSwitchParameter_1, output=""}
    Metallic {from=MaterialExpressionConstant_0, output=""}
    Specular {from=MaterialExpressionAdd_1, output=""}
    Roughness {from=MaterialExpressionSaturate_1, output=""}
    World Position Offset {from=MaterialExpressionLinearInterpolate_0, output=""}
   ]
   properties {attributeSetTypes [
     #2
     #3
     #4
     #5
     #1
    ]}
  }
  MaterialExpressionSetMaterialAttributes_1 {
   class=MaterialExpressionSetMaterialAttributes
   inputs [
    MaterialAttributes {from=MaterialExpressionGetMaterialAttributes_0, output=MaterialAttributes}
    World Position Offset {from=MaterialExpressionSubtract_6, output=""}
   ]
   properties {attributeSetTypes [#1]}
  }
  MaterialExpressionStaticBoolParameter_1 {class=MaterialExpressionStaticBoolParameter, inputs [], properties {defaultValue=true, parameterName=UseFlatShading}}
  MaterialExpressionStaticBool_0 {class=MaterialExpressionStaticBool}
  MaterialExpressionStaticSwitchParameter_0 {
   class=MaterialExpressionStaticSwitchParameter
   inputs [
    True {from=MaterialExpressionConstant_3, output=""}
    False {from=MaterialExpressionConstant_4, output=""}
   ]
   properties {parameterName=IsFrosting}
  }
  MaterialExpressionStaticSwitchParameter_1 {
   class=MaterialExpressionStaticSwitchParameter
   inputs [
    True {from=MaterialExpressionVectorParameter_0, output=RGB}
    False {from=MaterialExpressionStaticSwitchParameter_2, output=""}
   ]
   properties {parameterName=IsFrosting}
  }
  MaterialExpressionStaticSwitchParameter_2 {
   class=MaterialExpressionStaticSwitchParameter
   inputs [
    True {from=MaterialExpressionVectorParameter_2, output=RGB}
    False {from=MaterialExpressionCurveAtlasRowParameter_3, output=""}
   ]
   properties {parameterName=IsSprinkle}
  }
  MaterialExpressionSubtract_0 {
   class=MaterialExpressionSubtract
   inputs [
    A {from=MaterialExpressionTime_0}
    B {from=MaterialExpressionScalarParameter_0}
   ]
  }
  MaterialExpressionSubtract_1 {
   class=MaterialExpressionSubtract
   inputs [
    A {from=MaterialExpressionMaterialFunctionCall_0, output=Object Pivot Location}
    B {from=MaterialExpressionWorldPosition_0, output=XYZ}
   ]
  }
  MaterialExpressionSubtract_4 {
   class=MaterialExpressionSubtract
   inputs [
    A {from=MaterialExpressionWorldPosition_1, output=Z}
    B {from=MaterialExpressionComponentMask_1}
   ]
  }
  MaterialExpressionSubtract_5 {
   class=MaterialExpressionSubtract
   inputs [
    A {from=MaterialExpressionSubtract_4}
    B {from=MaterialExpressionRuntimeVirtualTextureSample_0, output=WorldHeight}
   ]
  }
  MaterialExpressionSubtract_6 {
   class=MaterialExpressionSubtract
   inputs [
    A {from=MaterialExpressionGetMaterialAttributes_0, output=World Position Offset}
    B {from=MaterialExpressionConvert_0}
   ]
  }
  MaterialExpressionSubtract_7 {
   class=MaterialExpressionSubtract
   inputs [
    A {from=MaterialExpressionComponentMask_2}
    B {from=MaterialExpressionWorldPosition_1, output=Z}
   ]
  }
  MaterialExpressionTime_0 {class=MaterialExpressionTime}
  MaterialExpressionTransformPosition_1 {
   class=MaterialExpressionTransformPosition
   inputs [None {from=MaterialExpressionConstant3Vector_1, output=""}]
   properties {transformType=TRANSFORMPOSSOURCE_World}
  }
  MaterialExpressionVectorParameter_0 {
   class=MaterialExpressionVectorParameter
   properties {bUseCustomPrimitiveData=true, parameterName=Frosting Color, primitiveDataIndex=2, defaultValue=<rgba 1 0 0.7992 1>}
  }
  MaterialExpressionVectorParameter_1 {class=MaterialExpressionVectorParameter, properties {parameterName=Crumb Color, defaultValue=<rgba 0.4031 0.2499 0.0939 1>}}
  MaterialExpressionVectorParameter_2 {class=MaterialExpressionVectorParameter, properties {parameterName=Sprinkle Color, defaultValue=<rgba 1 0 0.03 1>}}
  MaterialExpressionWorldPosition_0 {class=MaterialExpressionWorldPosition, inputs [], properties {}}
  MaterialExpressionWorldPosition_1 {class=MaterialExpressionWorldPosition, inputs [], properties {}}
 }
 properties {
  bAutomaticallySetUsageInEditor=false
  bUseMaterialAttributes=true
  bUsedWithInstancedStaticMeshes=true
  bUsedWithNiagaraMeshParticles=true
  usedShadingModels=Default Lit
 }
}
